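# WBP_BRGameUI_FPSTicker is a Battle Royale debug HUD element widget that shows the FPS ticker.
# It derives from FortHUDElementWidget and wraps the existing AthenaFPSTicker widget as its root, alongside a TutorialHighlight widget.
# The event graph is empty and there are no variables, so any behavior comes from the parent class and the nested widgets.
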WidgetBlueprint /BattleRoyaleGameUI/HUD/Debug/WBP_BRGameUI_FPSTicker
graphs [EventGraph {links [], nodes []}]
variables []
widgets {
 info {
  inheritedWidgetCount=1
  parentClass=&FortHUDElementWidget
  rootWidgetClass=/Game/Athena/HUD/AthenaFPSTicker.AthenaFPSTicker_C
  widgetCount=2
 }
 widgets [
  AthenaFPSTicker {widgetClassPath=/Game/Athena/HUD/AthenaFPSTicker, bIsVariable=true, widget=AthenaFPSTicker}
  TutorialHighlight {widgetClassPath=&TutorialHighlight}
 ]
}
